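# Physics asset for the Solar Theory pickaxe cosmetic, named as the front-end variant, that sets up physical motion for a hinged part of the axe.
# It holds two capsule bodies: a small one on base_hinge and a larger offset one on dyn_hinge00.
# A single constraint joins them as a narrow hinge, limiting swing to a few degrees on one axis while locking the other swing axis and twist.
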PhysicsAsset /BRCosmetics/Weapons/FORT_Melee/Pickaxe_SolarTheory/Meshes/SolarTheory_Axe_PhysicsFrontEnd
bodies {
 base_hinge [{shapeType=Capsule, length=0.5, radius=0.5, shapeName=base_hinge_capsule, rotation=<rot 0 -90 0>}]
 dyn_hinge00 [{shapeType=Capsule, length=5.5825, radius=2.3718, shapeName=dyn_hinge_capsule, center=<7 0 0>, rotation=<rot 90 0 0>}]
}
constraints [{
  bone1Name=dyn_hinge00
  bone2Name=base_hinge
  swing1LimitDegrees=7.5
  swing2LimitDegrees=0
  swing2Motion=Locked
  twistLimitDegrees=0
  twistMotion=Locked
 }]
