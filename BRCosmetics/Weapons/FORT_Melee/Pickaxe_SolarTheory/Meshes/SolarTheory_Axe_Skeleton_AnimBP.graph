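# Animation Blueprint for the Solar Theory pickaxe that adds rigid body physics motion to the axe.
# The cached input pose feeds two RigidBody simulations with separate physics assets, and bIsInFrontEnd blends between the front-end and gameplay results.
# An additive handle yaw follows, enabled while moving without sprinting, apparently to avoid clipping.
# The update event also lowers gameplay linear velocity and acceleration scales while moving or sprinting.
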AnimBlueprint /BRCosmetics/Weapons/FORT_Melee/Pickaxe_SolarTheory/Meshes/SolarTheory_Axe_Skeleton_AnimBP
target_skeleton=&/SolarTheory_Axe_Skeleton
graphs [
 AnimGraph {nodes [
   n1 = RigidBody!(
    ComponentPose=Local To Component(LocalPose=Use cached pose 'InputPose'())
    Alpha
    OverrideWorldGravity:Vector="0,0,-500"
    ExternalForce
    ComponentLinearAccScale:Vector="0.5,0.5,0.5"
    ComponentLinearVelScale:Vector="0.5,0.5,0.5"
   ) -> (Pose) {
    bOverrideWorldGravity=true
    overridePhysicsAsset=&/SolarTheory_Axe_PhysicsFrontEnd
    simulationSpace=BaseBoneSpace
    alphaBoolBlend {blendOption=Linear}
    baseBoneRef {boneName=b_AR_Root}
    componentLinearAccScale=<0.3 0.3 0.3>
    componentLinearVelScale=<0.3 0.3 0.3>
    simSpaceSettings {
     maxAngularAcceleration=1000
     maxAngularVelocity=3
     maxLinearAcceleration=700
     maxLinearVelocity=500
     velocityScaleZ=0
     worldAlpha=1
     externalLinearDragV=<0.2 0.2>
    }
   }
   n2 = Component To Local(ComponentPose=@n1)
   n3 = RigidBody(
    ComponentPose=Local To Component(LocalPose=Use cached pose 'InputPose'())
    OverrideWorldGravity="0,0,-800"
    ComponentLinearAccScale="Get  Linear Acc Scale setup"() -> (" Linear Acc Scale setup":Vector)
    ComponentLinearVelScale="Get  Linear Vel Scale setup"() -> (" Linear Vel Scale setup":Vector)
   ) {
    bOverrideWorldGravity=true
    overridePhysicsAsset=&/SolarTheory_Axe_Physics
    simulationSpace=BaseBoneSpace
    alphaBoolBlend {blendOption=Linear}
    baseBoneRef {boneName=Melee_TwoHanded_Handle}
    componentAppliedLinearAccClamp=<3000 3000 3000>
    componentLinearAccScale=<0.3 0.3 0.3>
    componentLinearVelScale=<0.3 0.3 0.3>
    simSpaceSettings {
     maxAngularAcceleration=1000
     maxAngularVelocity=3
     maxLinearAcceleration=700
     maxLinearVelocity=500
     velocityScaleZ=0
     worldAlpha=0.25
     externalLinearDragV=<0.2 0.2>
    }
   }
   n4 = Component To Local(ComponentPose=@n3)
   n5 = Get bIsInFrontEnd()
   n6 = "Transform (Modify) Bone - Bone: Melee_TwoHanded_Handle"(
    ComponentPose=Local To Component(LocalPose=Blend Poses by bool(BlendPose_0=@n2, BlendPose_1=@n4, bActiveValue=@n5) {blendType=Linear})
    Alpha=Get isClippingAlpha() -> (isClippingAlpha:float)
    Translation="0,0,0"
    Rotation="0,35,0"
    Scale="1,1,1"
   ) {
    rotationMode=BMM_Additive
    rotationSpace=BCS_BoneSpace
    alphaBoolBlend {blendOption=Linear}
    alphaScaleBiasClamp {bInterpResult=true, interpSpeedDecreasing=5}
    boneToModify {boneName=Melee_TwoHanded_Handle}
    rotation {yaw=35}
   }
   Output Pose(Result=Component To Local(ComponentPose=@n6)) {layerGroup=None}
   Save cached pose 'InputPose'(Pose=Input Pose() {name=InPose_0}) {cacheName=InputPose}
  ]}
 EventGraph {nodes [
   Event BlueprintUpdateAnimation()
   > n1 = Sequence()
   n2 = Get bIsSprinting()
   > n3 = Branch(Condition=AND Boolean(A=NOT Boolean(A=@n2), B=Get bIsMoving2D()))
   > Set isClippingAlpha(isClippingAlpha:float=1) -> (Output_Get:float)
   Set isClippingAlpha(execute=@n3.else, isClippingAlpha=0)
   n4 = Branch(execute=@n1.then_1, Condition=OR Boolean(A=Get bIsSprinting(), B=Get bIsMoving2D()))
   > n5 = "Set  Linear Acc Scale setup"(" Linear Acc Scale setup":Vector="0.1,0.1,0.1") -> (Output_Get:Vector)
   > "Set  Linear Vel Scale setup"(" Linear Vel Scale setup":Vector=@n5) -> (Output_Get:Vector)
   n6 = "Set  Linear Acc Scale setup"(execute=@n4.else, " Linear Acc Scale setup"="0.3,0.3,0.3")
   > "Set  Linear Vel Scale setup"(" Linear Vel Scale setup"=@n6)
   TryGetPawnOwner()
  ]}
]
variables [
 " Linear Acc Scale setup" {type=Vector, default=<0.3 0.3 0.3>}
 " Linear Vel Scale setup" {type=Vector, default=<0.3 0.3 0.3>}
 Gravity Override Params {type=GravityOverrideParamsStruct}
 GravityOverride {type=Vector}
 isClippingAlpha {type=number, default=1}
]
